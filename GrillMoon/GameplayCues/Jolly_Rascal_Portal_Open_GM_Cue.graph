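# Sound cue for the Jolly Rascal portal opening, playing a rift-open sound in the Gadgets Items sound class.
# A single wave player feeds the output directly at boosted volume, with no randomization or mixing.
# It overrides attenuation with natural-sound falloff and distance-based low-pass filtering.
# A concurrency override allows only a couple of simultaneous instances, stopping the farthest before preventing new ones.
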SoundCue /GrillMoon/GameplayCues/Jolly_Rascal_Portal_Open_GM_Cue
root=WavePlayer_0
nodes [{id=WavePlayer_0, properties {soundWaveAssetPtr=/FNGameplayCues/Athena/Sounds/Weapons/Jolly_Rascal/WPN_Jolly_Rascal_Rift_Open_01}}]
properties {
 attenuationSettings=/FNGameplayCues/Sounds/Attenuation/Fort_Jolly_Rascal_Att
 bOverrideAttenuation=true
 bOverrideConcurrency=true
 maxDistance=6000
 soundClassObject=&Fort_Gadgets_Items
 volumeMultiplier=1.5
 attenuationOverrides {distanceAlgorithm=NaturalSound, falloffDistance=3800, lPFRadiusMax=3000, lPFRadiusMin=2000, attenuationShapeExtents {x=2200}}
 concurrencyOverrides {maxCount=2, resolutionRule=StopFarthestThenPreventNew}
}
